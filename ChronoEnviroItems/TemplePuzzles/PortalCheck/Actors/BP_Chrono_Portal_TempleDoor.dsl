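; Temple vault door for the Chrono portal puzzle, opened by one randomly chosen key statue among nearby portal statues.
; After match ready and a delay, the server excludes child statues and links the key to the door, or opens it if the hotfix flag is off.
; Timelines slide both halves, with client-side dust, shake and audio.
; RepNotify clears collision and interaction; conversation tracking limits the music stinger to the interacting client.
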(event EventBeginPlay
  ;[1 1 1] Init
  ;;[1 1 1] TODO: playspace, this is temp for unblocking testing, delay to make sure, statues are loaded
  (Class|FortAsyncActionAuthorityMatchReady|MatchReadyAsync
    (:OnReady
      (Utilities|FlowControl|Delay 1e+01)
      (Default|SetStartandEndPositionsforDoors)
      (if (ScalableFloat|GetValueasBool (Variables|Default|GetHotFixEnabled))
        ;[1 0.987 0.9767] Grab all nearby teleporters, set one to be the key
        (Networking|SwitchHasAuthority
          (:Authority
            (bind _outactors (Class|SpecialEventHelpers|GetActorsinBox (Utilities|Array|MakeArray (Variables|Default|GetPortal%20Statue%20Actor%20Class)) (Math|Box|MakeBox (- (Transformation|GetActorLocation) (ScalableFloat|GetValueatLevel (Variables|Default|GetMaxPuzzleNearRadius))) (+ (Transformation|GetActorLocation) (ScalableFloat|GetValueatLevel (Variables|Default|GetMaxPuzzleNearRadius))))))
            (Default|SetRandomKeytoDoorfromPortalArray _outactors)))
        (else
          (CallFunction|OpenVaultDoor))))))

(event Custom|Open
  (bind _audioloop (Variables|Default|GetAudioLoop))
  (bind _returnvalue (Transformation|GetWorldLocation _audioloop))
  ;[1 1 1] Open Door
  (Utilities|FlowControl|DoOnce false
    (:Completed
      (Variables|Default|SetDoorOpened true)
      (Utilities|FlowControl|Sequence
        (:then_0
          (CallFunction|DoorAnimatingCosmetics) ; inc audio
        )
        (:then_1
          (bind (_direction _dooralpha) (Timeline|DoorOpenAnim PlayFromStart)
            (:Update
              (Niagara|SetNiagaraVariable(Float) (Variables|Default|GetNS_Dust) "DoorOpen" _dooralpha)
              (Transformation|SetWorldLocation (Variables|Default|GetSM_Ancients_VaultDoor_R) (Math|Vector|Lerp(Vector) (Variables|Default|GetStartLocation_R) (Variables|Default|GetEndLocation_R) _dooralpha) false false)
              (Transformation|SetWorldLocation (Variables|Default|GetSM_Ancients_VaultDoor_L) (Math|Vector|Lerp(Vector) (Variables|Default|GetStartLocation_L) (Variables|Default|GetEndLocation_L) _dooralpha) false false))
            (:EndShake
              (Utilities|FlowControl|Sequence
                (:then_0
                  (Camera|PlayWorldCameraShake (Variables|Default|GetShake) (Transformation|GetWorldLocation (Variables|Default|GetStaticMesh)) 2e+03 2e+03 0.75)
                  (Components|DestroyComponent (Variables|Default|GetBox_Nav_Middle)) ; Allow AI to pass when the door has been opened
                  (ForceFeedback|SpawnForceFeedbackatLocation (Variables|Default|GetForce%20Feedback%20Effect) (Transformation|GetWorldLocation (Variables|Default|GetStaticMesh))))
                (:then_1
                  ;;[0.05153 0.09729 0.1969] Audio
                  (Audio|SpawnSoundatLocation (Variables|Sound|GetDoorStopSound) _returnvalue)
                  (Audio|SpawnSoundatLocation (Variables|Sound|GetCamShakeSound) _returnvalue))))))
        (:then_2
          (if (Variables|Default|GetDebugResetState)
            (Utilities|FlowControl|Delay 1e+01)
            (CallFunction|CloseVaultDoor)))))))

(event Custom|OpenVaultDoor
  (bind _audioloop (Variables|Default|GetAudioLoop))
  (bind _returnvalue (Transformation|GetWorldLocation _audioloop))
  ;[1 1 1] Open Door
  (Utilities|FlowControl|DoOnce false
    (:Completed
      (Variables|Default|SetDoorOpened true)
      (Utilities|FlowControl|Sequence
        (:then_0
          (CallFunction|DoorAnimatingCosmetics) ; inc audio
        )
        (:then_1
          (Timeline|DoorOpenAnim PlayFromStart))
        (:then_2
          (if (Variables|Default|GetDebugResetState)
            (Utilities|FlowControl|Delay 1e+01)
            (CallFunction|CloseVaultDoor)))))))

(event Custom|CloseVaultDoor
  (bind _audioloop (Variables|Default|GetAudioLoop))
  (bind _returnvalue (Transformation|GetWorldLocation _audioloop))
  ;[1 1 1] Open Door
  (if (Variables|Default|GetDebugResetState)
    (Variables|Default|SetDoorOpened false)
    (else
      (Utilities|FlowControl|DoOnce false
        (:Completed
          (Variables|Default|SetDoorOpened true)))))
  (CallFunction|DoorAnimatingCosmetics)
  (bind (_direction _dooralpha) (Timeline|Timeline PlayFromStart)
    (:Update
      (Niagara|SetNiagaraVariable(Float) (Variables|Default|GetNS_Dust) "DoorOpen" _dooralpha)
      (Transformation|SetWorldLocation (Variables|Default|GetSM_Ancients_VaultDoor_R) (Math|Vector|Lerp(Vector) (Variables|Default|GetEndLocation_R) (Variables|Default|GetStartLocation_R) _dooralpha) false false)
      (Transformation|SetWorldLocation (Variables|Default|GetSM_Ancients_VaultDoor_L) (Math|Vector|Lerp(Vector) (Variables|Default|GetEndLocation_L) (Variables|Default|GetStartLocation_L) _dooralpha) false false))
    (:EndShake
      ;;[0.05153 0.09729 0.1969] Audio
      (Audio|SpawnSoundatLocation (Variables|Sound|GetDoorStopSound) _returnvalue)
      (Audio|SpawnSoundatLocation (Variables|Sound|GetCamShakeSound) _returnvalue))) ; door closing anim
)

(event Custom|DoorAnimatingCosmetics
  (bind _audioloop (Variables|Default|GetAudioLoop))
  (bind _returnvalue (Transformation|GetWorldLocation _audioloop))
  ;[1 1 1] FX and audio
  (Networking|SwitchHasAuthority
    (:Remote
      (Utilities|FlowControl|Sequence
        (:then_0
          (Camera|PlayWorldCameraShake (Variables|Default|GetShake) (Transformation|GetWorldLocation (Variables|Default|GetStaticMesh)) 2e+03 2e+03 0.75)
          (bind _returnvalue_2 (Niagara|SpawnSystemAttached "/ChronoEnviroItems/TemplePuzzles/RedLight/FX/NS_Chrono_Temple_DoorDust.NS_Chrono_Temple_DoorDust" (Variables|Default|GetStaticMesh) "None" (Transformation|GetWorldLocation (Variables|Default|GetStaticMesh)) "0, 0, 0" "KeepWorldPosition"))
          (Variables|Default|SetNS_Dust _returnvalue_2)
          (ForceFeedback|SpawnForceFeedbackatLocation (Variables|Default|GetForce%20Feedback%20Effect) (Transformation|GetWorldLocation (Variables|Default|GetStaticMesh))))
        (:then_1
          ;;[0.05153 0.09729 0.1969] Audio
          (Utilities|FlowControl|DoOnce false
            (:Completed
              (Audio|SpawnSoundatLocation (Variables|Sound|GetDoorOpenSound) _returnvalue)
              (Audio|SpawnSoundatLocation (Variables|Sound|GetDoorMovingSound) _returnvalue)
              (if (Variables|Default|GetDoorOpened)
                (bind _audioloop_2 (Variables|Default|GetAudioLoop)
                  (:then
                    (Audio|Components|Audio|FadeOut _audioloop_2 7.0 0.0 "Linear" :!class /Script/Engine.AudioComponent)
                    (Utilities|FlowControl|Delay 2.0)
                    (Utilities|FlowControl|Sequence
                      (:then_0
                        _)
                      (:then_1
                        _))
                    (Utilities|FlowControl|Delay 7.0)
                    (Utilities|FlowControl|DoOnce false
                      (:Completed
                        (Utilities|IsValid (Variables|Default|GetLocal%20Interacting%20Controller)
                          (:Is%20Valid
                            (Audio|SpawnSound2D (Variables|Sound|GetMusicStinger))
                            (Utilities|FlowControl|Delay 1.0))) ; Having this variable set means we're in the client that started interacting with this door.
                      )))
                  (:else
                    _)))
              (bind _audioloop_3 (Variables|Default|GetAudioLoop)
                (:then
                  (Audio|Components|Audio|FadeIn _audioloop_3 7.0 1.0 0.0 "Linear" :!class /Script/Engine.AudioComponent))
                (:else
                  _)))))))))

(event Custom|DestroyDoor
  (bind _self self)
  ;[1 1 1] Destroy
  (Variables|Default|SetDoorDestroyed true)
  (Networking|SwitchHasAuthority
    (:Authority
      (Utilities|IsValid _self
        (:Is%20Valid
          (Actor|DestroyActor _self))))
    (:Remote
      (Niagara|SpawnSystematLocation "/Game/Effects/Fort_Effects/World/Destruction/Base_Effects/NS_General_Deco_Destruction_Large_Converted.NS_General_Deco_Destruction_Large_Converted" (Transformation|GetWorldLocation (Variables|Default|GetDefaultSceneRoot)))
      (Rendering|SetActorHiddenInGame self true)
      (bind _audioloop (Variables|Default|GetAudioLoop)
        (:then
          (Components|DestroyComponent _audioloop))
        (:else)))))

(event OnReceivedLocalClientConversationMessage(NonPlayerConversationComponent)
  (params (LocalPlayerController :type Fort%20Player%20Controller%20Object%20Reference) (Message :type Client%20Conversation%20Message%20Payload%20Structure%20%28by%20ref%29))
  ;[0.05153 0.09729 0.1969] Audio For Text Prompts
  (Networking|SwitchHasAuthority
    (:Remote
      (bind _output_get (Variables|Default|SetLocal%20Interacting%20Controller LocalPlayerController))
      (Audio|SpawnSound2D (Variables|Sound|GetTextPromptOpenSound))
      (Ability|Tasks|WaitGameplayTagRemoveFromActor (Variables|/Script/Engine.Controller|GetPlayerState _output_get) "(TagName=\"Conversation.InteractingWithNPC\")" true)
      (CallFunction|OnClientFinishedConversation))))

(event Custom|OnClientFinishedConversation
  ;[0.05153 0.09729 0.1969] Audio For Text Prompts
  (Utilities|FlowControl|Sequence
    (:then_0
      (Audio|SpawnSound2D (Variables|Sound|GetTextPromptCloseSound)))
    (:then_1
      (Utilities|Time|ClearandInvalidateTimerbyHandle (Variables|Default|GetClientFinishesConversationTimer))
      (bind _returnvalue (Utilities|Time|SetTimerbyEvent (delegate Custom|CheckIfDoorOpenToUnsetController) 2.0))
      (Variables|Default|SetClientFinishesConversationTimer _returnvalue))))

(event Custom|CheckIfDoorOpenToUnsetController
  ;[0.05153 0.09729 0.1969] Audio For Text Prompts
  (if (not (Variables|Default|GetDoorOpened))
    (Variables|Default|SetLocal%20Interacting%20Controller 0)))

(fn BlueprintCanInteract
  (params (InteractingPawn :type Fort%20Pawn%20Object%20Reference) (InteractionBeingAttempted :type EInteractionBeingAttempted%20Enum :default "FirstInteraction") (InteractionType :type TInteractionType%20Enum :default "IT_NoInteraction"))
  (return true))

(fn BlueprintGetAllHighlightableComponents
  (return (Utilities|Array|MakeArray (Variables|Default|GetSM_Ancients_VaultDoor_L) (Variables|Default|GetSM_Ancients_VaultDoor_R)) true))

(fn BlueprintGetFailedInteractionString
  (params (InteractingPawn :type Fort%20Pawn%20Object%20Reference) (InteractionBeingAttempted :type EInteractionBeingAttempted%20Enum :default "FirstInteraction"))
  (return ""))

(fn BlueprintGetInteractionString
  (params (InteractingPawn :type Fort%20Pawn%20Object%20Reference) (InteractionBeingAttempted :type EInteractionBeingAttempted%20Enum :default "FirstInteraction"))
  (return (Variables|Default|GetInteractionString)))

(fn BlueprintGetInteractionTime
  (params (InteractingPawn :type Fort%20Pawn%20Object%20Reference) (InteractionBeingAttempted :type EInteractionBeingAttempted%20Enum :default "FirstInteraction"))
  (return 0.0 true))

(fn OnRep_DoorDestroyed
  (if (Variables|Default|GetDoorDestroyed)
    (Rendering|SetVisibility (Variables|Default|GetStaticMesh) false false)
    (Collision|SetCollisionEnabled (Variables|Default|GetBlockAllCollision) "NoCollision")
    (else
      (Rendering|SetVisibility (Variables|Default|GetStaticMesh) true false))))

(fn OnRep_DoorOpened
  (if (Variables|Default|GetDoorOpened)
    (Collision|SetCollisionEnabled (Variables|Default|GetBlockAllCollision) "NoCollision")
    (InteractionUI|AddHideUIIndicatorReason (Variables|Default|GetNonPlayerConversationComponent) (Variables|Default|GetHide%20Conversation%20UI%20Indicator%20Reason))
    (Variables|Interaction|SetbShowFirstInteractPrompt false)
    (Variables|Interaction|SetbAllowInteract false)
    (for _array_element (Variables|Default|GetPuzzle%20Portals%20Array)
      (Variables|/ChronoEnviroItems/TemplePuzzles/PortalCheck/Actors/B_Chrono_PortalCheck_Statue.B_Chrono_PortalCheck_Statue_C|SetStatue%20Interact%20Text (Variables|/ChronoEnviroItems/TemplePuzzles/PortalCheck/Actors/B_Chrono_PortalCheck_Statue.B_Chrono_PortalCheck_Statue_C|GetIncorrectStatueInteractionStringAfterInteract _array_element) _array_element))
    (else
      (Collision|SetCollisionEnabled (Variables|Default|GetBlockAllCollision) "QueryAndPhysics"))))

(fn SetRandomKeyToDoorFromPortalArray
  (params (Statue%20Array :type Array%20of%20Actor%20Object%20References))
  (for _array_element Statue%20Array
    ;[1 1 1] Grab one way portals for the door puzzle. Ignore child version for puzzles, these are for rifts only
    (bind _asb_chrono_portal_check_statue (Utilities|Casting|CastToB_Chrono_PortalCheck_Statue _array_element))
    (Utilities|Casting|CastToB_Chrono_PortalCheck_Statue_Child _asb_chrono_portal_check_statue
      (:CastFailed
        (Utilities|Array|AddUnique (Variables|Default|GetPuzzle%20Portals%20Array) _asb_chrono_portal_check_statue))))
  ;[1 1 1] Set one portal to be the door key trigger, set the door reference so we can open it later, remove teleporter stuff from it
  (bind _element (ArrayMacros|GetRandomArrayElement (Variables|Default|GetPuzzle%20Portals%20Array))
    (:%20%20
      (Variables|/ChronoEnviroItems/TemplePuzzles/PortalCheck/Actors/B_Chrono_PortalCheck_Statue.B_Chrono_PortalCheck_Statue_C|SetIs%20Door%20Key%20Portal true _element)
      (Variables|/ChronoEnviroItems/TemplePuzzles/PortalCheck/Actors/B_Chrono_PortalCheck_Statue.B_Chrono_PortalCheck_Statue_C|SetLinked%20Temple%20Portal%20Door self _element))
    (:Empty%20Array
      (Development|PrintString "Chrono Temple Portal Array is Empty. There are no portal actors in the radius of the door!!!"))))

(fn SetStartAndEndPositionsForDoors
  (bind _sm_ancients_vaultdoor_l (Variables|Default|GetSM_Ancients_VaultDoor_L))
  (bind _returnvalue (Transformation|GetWorldLocation _sm_ancients_vaultdoor_l))
  (bind _sm_ancients_vaultdoor_r (Variables|Default|GetSM_Ancients_VaultDoor_R))
  (bind _returnvalue_2 (Transformation|GetWorldLocation _sm_ancients_vaultdoor_r))
  ;[1 1 1] Store off base 'start' position. Right door needs to go 'backwards' negative units
  (Variables|Default|SetStartLocation_R _returnvalue_2)
  (Variables|Default|SetEndLocation_R (+ _returnvalue_2 (* (Transformation|GetForwardVector _sm_ancients_vaultdoor_r) (Math|Conversions|ToVector(Float) (Variables|Default|GetNegativeDoorExtensionLength)))))
  ;[1 1 1] Store off base 'start' position. Left door needs to go 'forwards' positive units
  (Variables|Default|SetStartLocation_L _returnvalue)
  (Variables|Default|SetEndLocation_L (+ _returnvalue (* (Transformation|GetForwardVector _sm_ancients_vaultdoor_l) (Math|Conversions|ToVector(Float) (Variables|Default|GetDoorExtensionLength)))))
  (return))

(fn ConstructionScript)
