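; Interactable temple puzzle statue that either opens its linked portal door as the solution or rifts the interacting player away.
; Both paths play caught cosmetics and delay; wrong statues then, on authority, force-teleport through a spawned rift, reusing it until a timer destroys it.
; A replicated state enum drives eye VFX and interaction shutdown.
; Hotfix values control delays, rift lifetime, repeat use and disabling the statue at BeginPlay.
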(event EventBeginPlay
  ;[1 1 1] If hotfix disabled, shut down statues ( solution statue does this on interact, so we use this )
  (if (not (ScalableFloat|GetValueasBool (Variables|Default|GetHotFixEnabled)))
    (Variables|Default|SetEnumState "NewEnumerator1")))

(event Interaction|EventBlueprintOnInteract
  (params (InteractingPawn :type Fort%20Pawn%20Object%20Reference) (InteractionBeingAttempted :type EInteractionBeingAttempted%20Enum :default "FirstInteraction"))
  (bind (_returnvalue_location _returnvalue_rotation) (Math|Transform|BreakTransform (Transformation|GetWorldTransform (Variables|/Script/Engine.Character|GetMesh (Variables|Default|GetInteracting%20Pawn)))))
  (bind _returnvalue_location_2 (Math|Transform|BreakTransform (Transformation|GetWorldTransform (Variables|/Script/Engine.Character|GetMesh (Variables|Default|GetInteracting%20Pawn)))))
  (bind _currentriftactor (Variables|Default|GetCurrentRiftActor))
  ;[1 1 1] If we're the door key, open, else AHA! rift the player
  (Variables|Default|SetInteracting%20Pawn InteractingPawn)
  (if (Variables|Default|GetIs%20Door%20Key%20Portal)
    (Utilities|FlowControl|Sequence
      (:then_0
        (CallFunction|CaughtCosmetics)
        (Utilities|FlowControl|Delay (ScalableFloat|GetValueatLevel (Variables|Default|GetDelayBeforeRiftPlayer)))
        (Class|BPChronoPortalTempleDoor|OpenVaultDoor (Variables|Default|GetLinked%20Temple%20Portal%20Door)))
      (:then_1
        (Variables|Default|SetEnumState "NewEnumerator1")))
    (else
      (Utilities|FlowControl|Sequence
        (:then_0
          (if (Variables|Default|GetbAllowMultipleRifts)
            (GameplayCue|ExecuteGameplayCueOnActor(Burst) (Variables|Default|GetInteracting%20Pawn) (Variables|Default|GetPreRift%20Gameplay%20Cue%20Tag) (Ability|GameplayCue|MakeGameplayCueParameters 0.0))
            (if (not (ScalableFloat|GetValueasBool (Variables|Default|GetAllowMultipleRiftUses)))
              (Variables|Default|SetbAllowMultipleRifts false))
            (CallFunction|CaughtCosmetics)
            (Utilities|FlowControl|Delay (ScalableFloat|GetValueatLevel (Variables|Default|GetDelayBeforeRiftPlayer)))
            (Components|Movement|StopMovementImmediately (Variables|/Script/Engine.Character|GetCharacterMovement (Variables|Default|GetInteracting%20Pawn)))
            (Networking|SwitchHasAuthority
              (:Authority
                ;[0 0.2204 1] If we already have a rift spawned, move it & use that ref to rift the player. Only spawn a new BGA if first time or previous destroyed
                (Utilities|IsValid (Variables|Default|GetCurrentRiftActor)
                  (:Is%20Valid
                    (Transformation|SetActorLocation _currentriftactor _returnvalue_location_2 false true)
                    (Class|BGARiftPortalItemAthena|ForceRiftTeleport _currentriftactor (Variables|Default|GetInteracting%20Pawn)))
                  (:Is%20Not%20Valid
                    ;[1 1 1] Hack: try to force teleport, spawning rifts on moving player is not great...
                    (bind _returnvalue (Game|SpawnActorfromClass :Class "/Game/Athena/Items/Consumables/RiftItem/BGA_RiftPortal_Item_Athena.BGA_RiftPortal_Item_Athena_C" :SpawnTransform (Math|Transform|MakeTransform _returnvalue_location _returnvalue_rotation) :CollisionHandlingOverride "AdjustIfPossibleButAlwaysSpawn" :TransformScaleMethod "OverrideRootScale"))
                    (Class|BGARiftPortalItemAthena|ForceRiftTeleport _returnvalue (Variables|Default|GetInteracting%20Pawn))
                    (Variables|Default|SetCurrentRiftActor _returnvalue)
                    (bind _returnvalue_2 (Utilities|Time|SetTimerbyEvent (delegate Custom|DestroyRiftEvent) (ScalableFloat|GetValueatLevel (Variables|Default|GetRiftLifeTime))))
                    (Variables|Default|SetDestroyRiftEventTimer _returnvalue_2)))))))
        (:then_1
          (Utilities|FlowControl|DoOnce false
            (:Completed
              (Variables|Default|SetEnumState "NewEnumerator2"))))))))

(event Custom|DestroyRiftEvent
  (Class|BGARiftPortalItemAthena|DestroyRift (Variables|Default|GetCurrentRiftActor)))

(event Custom|CaughtCosmetics
  (GameplayCue|ExecuteGameplayCue self (Variables|Default|GetCaught%20Gameplay%20Cue%20Tag))
  (Components|Activation|Activate (Variables|Default|GetNS_Burst_L) false))

(fn BlueprintCanInteract
  (params (InteractingPawn :type Fort%20Pawn%20Object%20Reference) (InteractionBeingAttempted :type EInteractionBeingAttempted%20Enum :default "FirstInteraction") (InteractionType :type TInteractionType%20Enum :default "IT_NoInteraction"))
  (return true))

(fn BlueprintGetFailedInteractionString
  (params (InteractingPawn :type Fort%20Pawn%20Object%20Reference) (InteractionBeingAttempted :type EInteractionBeingAttempted%20Enum :default "FirstInteraction"))
  (return ""))

(fn BlueprintGetInteractionString
  (params (InteractingPawn :type Fort%20Pawn%20Object%20Reference) (InteractionBeingAttempted :type EInteractionBeingAttempted%20Enum :default "FirstInteraction"))
  (return (Variables|Default|GetStatue%20Interact%20Text)))

(fn GetDebugString
  (return (Utilities|String|Append "Is Solution Statue: " (Utilities|String|ToString(Boolean) (Variables|Default|GetIs%20Door%20Key%20Portal)))))

(fn GetActorForInteraction
  (return 0))

(fn OnRep_EnumState
  (bind _ns_idle_r (Variables|Default|GetNS_Idle_R))
  (bind _ns_idle_r_2 (Variables|Default|GetNS_Idle_R))
  (switch Utilities|FlowControl|Switch|SwitchonChronoPortalStatueState_Enum (Variables|Default|GetEnumState)
    (:NewEnumerator1
      (Variables|Default|SetCan%20Interact false)
      (Variables|Interaction|SetbShowFirstInteractPrompt false)
      (Variables|Interaction|SetbShowInteractKeybind false)
      (Components|Activation|Activate _ns_idle_r_2 false)
      (Niagara|SetNiagaraVariable(Int32) _ns_idle_r_2 "FXColor" 1))
    (:NewEnumerator2
      ;[1 1 1] Set Idle to 0 for correct statue eyes color VFX
      (Components|Activation|Activate _ns_idle_r false)
      (Niagara|SetNiagaraVariable(Int32) _ns_idle_r "FXColor" 0)
      (if (not (ScalableFloat|GetValueasBool (Variables|Default|GetAllowMultipleRiftUses)))
        (Variables|Default|SetCan%20Interact false)
        (Variables|Interaction|SetbShowFirstInteractPrompt false)
        (Variables|Interaction|SetbShowInteractKeybind false)))))

(fn ConstructionScript
  (|CallParentFunction))
