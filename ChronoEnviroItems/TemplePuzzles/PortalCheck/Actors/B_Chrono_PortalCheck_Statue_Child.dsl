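; Child statue Blueprint for the Chrono temple puzzle portal check, apparently the rifting variant of the statue.
; On BeginPlay it reads its own hotfix ScalableFloat, separate from other statues, and switches its enum state to NewEnumerator1 when that flag is off.
; The interaction string override calls the parent but returns this Blueprint's interaction string variable.
; Overlap and Tick events are empty, and construction only calls the parent.
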(event EventBeginPlay
  ;[1 1 1] Separate hotfix for child rifting statues
  (if (not (ScalableFloat|GetValueasBool (Variables|Default|GetHotFixEnabled_0)))
    (Variables|Default|SetEnumState "NewEnumerator1")))

(event Collision|EventActorBeginOverlap
  (params (OtherActor :type Actor%20Object%20Reference)))

(event EventTick
  (params (DeltaSeconds :type Float%20%28single-precision%29)))

(fn BlueprintGetInteractionString
  (params (InteractingPawn :type Fort%20Pawn%20Object%20Reference) (InteractionBeingAttempted :type EInteractionBeingAttempted%20Enum :default "FirstInteraction"))
  (|CallParentFunction InteractingPawn InteractionBeingAttempted)
  (return (Variables|Default|GetInteractionString_0)))

(fn ConstructionScript
  (|CallParentFunction))
